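; Parent Blueprint for a challenge quest station prop whose lights and fan show whether the local player has a challenge active.
; Construction destroys it in disallowed playlists, otherwise creates dynamic materials and caches lit colors.
; A gameplay tag check sets the on look at client BeginPlay and switches on/off when the service is used.
; A cosmetic event plays a Niagara burst and sound, with a looping sound indicator stopped by retriggerable delay.
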(event EventBeginPlay
  (bind _returnvalue (Fortnite|GetFirstLocalFortPlayerController true))
  (bind _material_lightscontrol (Variables|Art|GetMaterial_LightsControl))
  (|CallParentFunction)
  ;[1 1 1] Update the actor's state if the player has a challenge active
  (Networking|SwitchHasAuthority
    (:Remote
      (Utilities|IsValid _returnvalue
        (:Is%20Valid
          (if (GameplayTags|HasMatchingGameplayTag (Ability|GetAbilitySystemComponent (Variables|/Script/Engine.Controller|GetPlayerState _returnvalue)) (Variables|Art|GetTag%20to%20Check))
            (Rendering|Material|SetVectorParameterValue _material_lightscontrol "SFX_Emissive_Tint(Expand)_3" (Variables|Art|GetWhiteLightR_On) :!class /Script/Engine.MaterialInstanceDynamic)
            (Rendering|Material|SetVectorParameterValue _material_lightscontrol "SFX_Emissive_Tint(Expand)_2" (Variables|Art|GetWhiteLightL_On) :!class /Script/Engine.MaterialInstanceDynamic)
            (Rendering|Material|SetVectorParameterValue _material_lightscontrol "SFX_Emissive_Tint(Expand)_4" (Variables|Art|GetOrangeLights_Off) :!class /Script/Engine.MaterialInstanceDynamic)
            (Rendering|Material|SetScalarParameterValue (Variables|Art|GetMatierial%20Fan%20Spin) "FanRotationSpeed" (Variables|Art|GetFanSpeed_On) :!class /Script/Engine.MaterialInstanceDynamic)))))))

(event EventEndPlay
  (params (EndPlayReason :type EEndPlayReason%20Enum :default "Destroyed"))
  (|CallParentFunction EndPlayReason)
  (SoundIndicator|StopSoundIndicatorLooping (Variables|Default|GetFortSoundIndicator)))

(event Custom|PlayBigAudio
  (bind _staticmeshcomponent (Variables|Debug|GetStaticMeshComponent))
  (bind _relativelocation (Variables|/Script/Engine.SceneComponent|GetRelativeLocation _staticmeshcomponent))
  ;[1 1 1] Play replicated FX
  (Networking|CanExecuteCosmeticEvents
    (:True
      (Niagara|SpawnSystematLocation "/ChallengeQuests/FX/NS_Quest_S30_Wasteland_Burst.NS_Quest_S30_Wasteland_Burst" (+ (Variables|/Script/Engine.SceneComponent|GetRelativeLocation (Variables|Debug|GetStaticMeshComponent)) "0.000000,0.000000,170.000000") (Math|Rotator|CombineRotators (Math|Conversions|RotationFromXVector (Transformation|GetForwardVector (Variables|Debug|GetStaticMeshComponent))) "0.000000,-90.000000,0.000000"))
      (Audio|PlaySoundatLocation (Variables|Audio|GetBigAudio) _relativelocation)
      ;[1 1 1] AV - adding looping indicator to cover the duration of the audio
      (SoundIndicator|StartSoundIndicatorLooping (Variables|Default|GetFortSoundIndicator) _relativelocation (Variables|Audio|GetBigAudio) 1.0 0)
      (Utilities|FlowControl|RetriggerableDelay 3.3)
      (SoundIndicator|StopSoundIndicatorLooping (Variables|Default|GetFortSoundIndicator)))))

(fn BlueprintCanInteract
  (params (InteractingPawn :type Fort%20Pawn%20Object%20Reference) (InteractionBeingAttempted :type EInteractionBeingAttempted%20Enum :default "FirstInteraction") (InteractionType :type TInteractionType%20Enum :default "IT_NoInteraction"))
  (return (|CallParentFunction InteractingPawn InteractionBeingAttempted InteractionType)))

(fn PlayServiceUsedEffects
  (params (PlayerController :type Fort%20Player%20Controller%20Object%20Reference))
  (bind _material_lightscontrol (Variables|Art|GetMaterial_LightsControl))
  (bind _material_lightscontrol_2 (Variables|Art|GetMaterial_LightsControl))
  (|CallParentFunction PlayerController)
  (if (not (GameplayTags|HasMatchingGameplayTag (Ability|GetAbilitySystemComponent PlayerController) (Variables|Art|GetTag%20to%20Check)))
    ;[1 1 1] Statue On State
    (Rendering|Material|SetVectorParameterValue _material_lightscontrol_2 "SFX_Emissive_Tint(Expand)_3" (Variables|Art|GetWhiteLightR_On) :!class /Script/Engine.MaterialInstanceDynamic)
    (Rendering|Material|SetVectorParameterValue _material_lightscontrol_2 "SFX_Emissive_Tint(Expand)_2" (Variables|Art|GetWhiteLightL_On) :!class /Script/Engine.MaterialInstanceDynamic)
    (Rendering|Material|SetVectorParameterValue _material_lightscontrol_2 "SFX_Emissive_Tint(Expand)_4" (Variables|Art|GetOrangeLights_Off) :!class /Script/Engine.MaterialInstanceDynamic)
    (Rendering|Material|SetScalarParameterValue (Variables|Art|GetMatierial%20Fan%20Spin) "FanRotationSpeed" (Variables|Art|GetFanSpeed_On) :!class /Script/Engine.MaterialInstanceDynamic)
    (else
      ;[1 1 1] Statue Off State
      (Variables|Art|SetMaterial_LightsControl (Class|BTechArtGlobalLibrary|GetDynamicMaterialInstance (Variables|Default|GetSM_Commerce_SidewalkSign_C)))
      (Rendering|Material|SetVectorParameterValue _material_lightscontrol "SFX_Emissive_Tint(Expand)_3" (Variables|Art|GetWhiteLightR_Off) :!class /Script/Engine.MaterialInstanceDynamic)
      (Rendering|Material|SetVectorParameterValue _material_lightscontrol "SFX_Emissive_Tint(Expand)_2" (Variables|Art|GetWhiteLIghtL_Off) :!class /Script/Engine.MaterialInstanceDynamic)
      (Rendering|Material|SetVectorParameterValue _material_lightscontrol "SFX_Emissive_Tint(Expand)_4" (Variables|Art|GetOrangeLights_On) :!class /Script/Engine.MaterialInstanceDynamic)
      (Rendering|Material|SetScalarParameterValue (Variables|Art|GetMatierial%20Fan%20Spin) "FanRotationSpeed" (Variables|Art|GetFanSpeed_Off) :!class /Script/Engine.MaterialInstanceDynamic))))

(fn ConstructionScript
  (bind _material_lightscontrol (Variables|Art|GetMaterial_LightsControl))
  ;[1 1 1] Initial playlist check
  (|CallParentFunction)
  (if (GameplayTags|HasAnyTags (Class|FortGameStateAthena|GetAthenaPlaylistContextTags (Fortnite|Athena|GameState|GetGameStateAthena)) (Variables|Default|GetT_DisallowedPlaylists))
    (Actor|DestroyActor)
    (else
      ;[1 1 1] Setting Initial values
      (Variables|Art|SetMaterial_LightsControl (Class|BTechArtGlobalLibrary|GetDynamicMaterialInstance (Variables|Default|GetSM_Commerce_SidewalkSign_C)))
      (Variables|Art|SetMatierial%20Fan%20Spin (Class|BTechArtGlobalLibrary|GetDynamicMaterialInstance (Variables|Default|GetSM_Commerce_SidewalkSign_C) 1))
      (bind _returnvalue (Rendering|Material|GetVectorParameterValue _material_lightscontrol "SFX_Emissive_Tint(Expand)_3" :!class /Script/Engine.MaterialInstanceDynamic))
      (Variables|Art|SetWhiteLightR_On _returnvalue)
      (bind _returnvalue_2 (Rendering|Material|GetVectorParameterValue _material_lightscontrol "SFX_Emissive_Tint(Expand)_2" :!class /Script/Engine.MaterialInstanceDynamic))
      (Variables|Art|SetWhiteLightL_On _returnvalue_2)
      (bind _returnvalue_3 (Rendering|Material|GetVectorParameterValue _material_lightscontrol "SFX_Emissive_Tint(Expand)_4" :!class /Script/Engine.MaterialInstanceDynamic))
      (Variables|Art|SetOrangeLights_On _returnvalue_3)
      ;[1 1 1] Turning lights off initially
      (Rendering|Material|SetVectorParameterValue _material_lightscontrol "SFX_Emissive_Tint(Expand)_3" (Variables|Art|GetWhiteLightR_Off) :!class /Script/Engine.MaterialInstanceDynamic)
      (Rendering|Material|SetVectorParameterValue _material_lightscontrol "SFX_Emissive_Tint(Expand)_2" (Variables|Art|GetWhiteLIghtL_Off) :!class /Script/Engine.MaterialInstanceDynamic))))
